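# Masked, two-sided subsurface material for Niagara mesh leaf particles in the Fall aura, drawing alder leaves that curl and burn away.
# Base color tints the leaf texture with per-particle hue and value variation, optionally using particle color or a pre-colorized texture.
# Dynamic parameters drive a sine-based world position offset curl and a randomly offset smoke-noise threshold.
# That threshold erodes the opacity mask, darkens the leaf and lights a hot emissive edge.
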Material /SparksCosmetics/Items/Auras/Assets/NS_Aura_Fall/M_FX_AlderLeaf_Burnaway
graph {
 properties {
  bUsedWithNiagaraMeshParticles=true
  blendMode=BLEND_Masked
  pixelDepthOffsetMode=PDOM_Legacy
  shadingModel=MSM_Subsurface
  twoSided=true
  usedShadingModels=Subsurface
 }
 nodes [
  n1 = TextureCoordinate()
  ComponentMask(None=@n1) {g=true}
  n2 = NamedRerouteUsage()
  LinearInterpolate(A=NamedRerouteUsage(), B=Constant3Vector() {constant=<rgba 0.0234 0.004 0 1>}, Alpha=@n2)
  n3 = TextureSampleParameter2D() {parameterName=OSSR Texture, samplerType=SAMPLERTYPE_Masks, texture=/Game/Environments/Asteria/Foliage/Trees/Oak/Textures/T_AlderLeafs2_OSSR}
  LinearInterpolate(B=ScalarParameter() {defaultValue=1, parameterName=Max Specular}, Alpha=@n3.R)
  LinearInterpolate(B=ScalarParameter() {defaultValue=1, parameterName=Max Roughness}, Alpha=@n3.B)
  n4 = VectorParameter() {parameterName=Hot Edge Color, defaultValue=<rgba 25 7.9348 0.9904 1>}
  MaterialFunctionCall(Emissive=Multiply(A=Multiply(A=@n4.RGB, B=@n4.RGB), B=@n2)) {materialFunction=&MF_FX_EmissiveAdjustments}
  n5 = TextureSampleParameter2D() {parameterName=Mask Texture, samplerType=SAMPLERTYPE_Masks, texture=/Game/Global/Textures/Default/Blanks/AllRed}
  n6 = ScalarParameter() {defaultValue=1, parameterName=Alpha or RBG Mask Lerp}
  n7 = StaticSwitchParameter(
   True=ChannelMaskParameter(None=LinearInterpolate(A=@n5.A, B=@n5.A, Alpha=@n6)) {parameterName=Mask Texture Channel}.R
   False=Constant() {r=1}
  ) {parameterName=Has Mask Texture}
  Multiply(A=Multiply(A=@n3.G) {constB=0.9, desc=Estimated value from Ateria Leaves Base Material}, B=@n7)
  Multiply(A=NamedRerouteUsage()) {constB=0.2, desc=Estimated value from Ateria Leaves Base Material}
  Multiply(A=NamedRerouteUsage(), B=@n7)
  n8 = ParticleRandom()
  n9 = ScalarParameter() {defaultValue=1, parameterName=DiffuseLeafTextureStrength}
  n10 = TextureSampleParameter2D() {parameterName=Diffuse Leaf Texture, texture=/Game/Environments/Asteria/Foliage/Trees/Oak/Textures/T_AlderLeafs2_D2}
  n11 = VectorParameter() {parameterName=LeafColor1, defaultValue=<rgba 0.1413 0.2052 0.0366 1>}
  n12 = ParticleColor()
  n13 = ScalarParameter() {parameterName=Use Particle Color}
  n14 = MaterialFunctionCall(
   Hue Shift Percentage=Add(A=Multiply(A=@n8) {constA=0.0125, constB=0.0125}, B=ScalarParameter() {defaultValue=-0.02, parameterName=Hue})
   Texture=Multiply(A=Multiply(A=@n9, B=@n10.RGB), B=LinearInterpolate(A=@n11.RGB, B=@n12.RGB, Alpha=@n13))
  ) {materialFunction=&HueShift}
  n15 = Add(A=Multiply(A=ScalarParameter() {defaultValue=0.1, parameterName=Value}, B=ParticleRandom()) {constA=0.1})
  NamedRerouteDeclaration(None=LinearInterpolate(
    A=Multiply(A=@n14.Result, B=@n15)
    B=TextureSampleParameter2D() {parameterName=Diffuse Leaf Texture_Color, texture=/Game/Environments/Asteria/Foliage/Trees/Oak/Textures/T_AlderLeafs2_D}.RGB
    Alpha=ScalarParameter() {parameterName=Use Colorized Texture}
   )) {name=Diffuse, nodeColor=<rgba 1 0.4688 0 1>}
  n16 = DynamicParameter() {parameterIndex=1, defaultValue=<rgba 0.3 3 1 0.01>, paramNames [Curl Offset, WPO Strength, Curl Scale, ""]}
  n17 = Multiply(A=@n16.Curl Scale, B=ComponentMask(None=Add(A=@n16.Curl Offset, B=@n1)) {g=true}) {constA=0.5}
  n18 = Multiply(
   A=@n16.WPO Strength
   B=MaterialFunctionCall(Z=Sine(None=@n17)) {materialFunction=&MakeFloat3}.Result
  )
  NamedRerouteDeclaration(None=Multiply(A=Transform(None=@n18) {transformSourceType=TRANSFORMSOURCE_Instance}, B=ScalarParameter() {defaultValue=1, parameterName=WPO Strength})) {name=WPO, nodeColor=<rgba 1 0 0.4219 1>}
  n19 = DynamicParameter() {defaultValue=<rgba 0.3 3 0.9 0.01>, paramNames ["", "", Burnaway, Burnaway Thickness]}
  n20 = Constant() {r=1}
  n21 = Add(A=OneMinus(None=@n19.Burnaway)) {constB=0.015}
  n22 = ParticleRandom()
  n23 = Add(A=TextureCoordinate(), B=AppendVector(A=@n22, B=@n22))
  n24 = AppendVector(A=ScalarParameter() {defaultValue=1, parameterName=Tile U Noise}, B=ScalarParameter() {defaultValue=1, parameterName=Tile V Noise})
  n25 = LinearInterpolate(Alpha=TextureSample(UVs=Multiply(A=@n23, B=@n24)) {texture=&T_Inky_Smoke_Tile}.G) {constA=0.1}
  NamedRerouteDeclaration(None=If(A=@n19.Burnaway, B=@n20, "A > B"=@n20, "A == B"=@n20, "A < B"=SmoothStep(Min=@n21, Max=@n21, Value=@n25))) {name=BurnawayOpacity, nodeColor=<rgba 0.1016 1 0 1>}
  n26 = Constant()
  n27 = SmoothStep(Min=@n21, Max=Add(A=@n21, B=Max(A=@n19.Burnaway Thickness) {constB=0.001}), Value=@n25)
  NamedRerouteDeclaration(None=If(A=@n19.Burnaway, B=@n20, "A > B"=@n26, "A == B"=@n26, "A < B"=OneMinus(None=@n27))) {name=BurnawayEmissive, nodeColor=<rgba 1 0 0.4219 1>}
  NamedRerouteUsage()
  TextureSampleParameter2D() {parameterName=Normal Texture, samplerType=SAMPLERTYPE_Normal, texture=/Game/Environments/Asteria/Foliage/Trees/Oak/Textures/T_AlderLeafs2_N}
 ]
}
