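# Niagara start burst for the Fall aura cosmetic, throwing autumn leaves and glowing sparks off the character.
# Short-lived Corona and Flare sprites track a skeletal mesh socket, while a separate spark burst swirls on curl noise.
# Colored mesh leaves spawn from sampled bones or sockets, then tumble and fall with animated material parameters.
# LeafEmbers reads those leaves to spawn trailing embers, randomly culled by the parent leaf's normalized age.
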NiagaraSystem /SparksCosmetics/Items/Auras/Assets/NS_Aura_Fall/NS_Aura_Fall_Start
dependencies {
 usedDataInterfaces [
  &NiagaraDataInterfaceCurve
  &NiagaraDataInterfaceParticleRead
  &NiagaraDataInterfaceArrayFloat
  &NiagaraDataInterfaceArrayColor
  &NiagaraDataInterfaceArrayInt32
  &NiagaraDataInterfaceSkeletalMesh
 ]
 usedDynamicInputs [&SelectFloatFromArray]
 usedModules [
  &self:SystemSpawnScript.SystemScriptSource.SystemScriptGraph.NiagaraNodeAssignment_0.SetVariables_#1
  &SystemState
 ]
}
emitters {
 Corona {
  emitterUpdateScript [
   &EmitterState(Life Cycle Mode=::NewEnumerator1, Loop Behavior=::NewEnumerator1, Loop Duration=1, UseLoopDelay=false)
   &SpawnBurst_Instantaneous(Spawn Count=1, Use Spawn Probability=false)
  ]
  particleSpawnScript [
   &InitializeParticle(
    Color=&MakeCustomLinearColorFromBool(Bool=$User.IsFrontEnd, True Color=<rgba 0.7 0.2226 0.057 1>, False Color=<rgba 2 0.636 0.1627 1>)
    Lifetime=0.35
    Sprite Rotation Mode=::NewEnumerator2
    Sprite Size Mode=::NewEnumerator3
    Uniform Sprite Size=400
    UsePositionOffset=false
    Write Material Random=true
   )
   &DynamicMaterialParameters(
    Index 0 Param 1=&RandomRangeFloat(Minimum=0.4, Maximum=0.6)
    Index 0 Param 2=1
    Index 0 Param 3=1
    Index 0 Param 4=1
    Module.Param0WriteEnabled=true
    Module.Param1WriteEnabled=true
    Module.Param2WriteEnabled=true
    Module.Param3WriteEnabled=true
   )
   &SystemLocation(MaskSpawnGroup=false, Offset=<0 0 120>)
  ]
  particleUpdateScript [
   &ParticleState()
   &SolveForcesAndVelocity(Clamp Velocity=false, Limit Acceleration=false)
   &SkeletalMeshLocation(Bone / Socket Index=7, Bone Sampling Mode=::NewEnumerator4, MaskSpawnGroup=false, Skeletal Mesh=$System.Skeletal Mesh) {module=SkeletalMeshLocation001}
   &ScaleColor(
    Scale Alpha=&Multiply_Float(
     A=&FNTimeOfDayScalar(Night Scale=0.7, Morning Scale=0.3, Evening Scale=0.4)
     B=&FloatFromCurve(FloatCurve=DataInterface() {propertyValues {Curve {keys [
          {interpMode=RCIM_Cubic, tangentMode=RCTM_User, time=0, value=0}
          {interpMode=RCIM_Cubic, tangentMode=RCTM_User, time=0.2, value=1}
          {interpMode=RCIM_Cubic, tangentMode=RCTM_User, time=1, value=0}
         ]}}})
    )
    Scale RGB=&Multiply_VectorByFloat(Vector=<1 1 1>, Float=1)
    ScaleA=true
    ScaleRGB=true
   )
   &ScaleSpriteSize(Scale Factor=&Vector2DFromFloat(Value=&FloatFromCurve(FloatCurve=DataInterface() {propertyValues {Curve {keys [
          {arriveTangent=2.1722, interpMode=RCIM_Cubic, leaveTangent=2.1722, tangentMode=RCTM_User, time=0, value=0}
          {arriveTangent=0.0224, interpMode=RCIM_Cubic, leaveTangent=0.0224, tangentMode=RCTM_User, time=1, value=1}
         ]}}})))
   &CameraOffset(Camera Offset Amount=-30)
   &DynamicMaterialParameters(
    Index 0 Param 1=0.75
    Index 0 Param 2=2
    Index 0 Param 3=1
    Index 0 Param 4=1
    Module.Param0WriteEnabled=true
    Module.Param1WriteEnabled=true
    Module.Param2WriteEnabled=true
    Module.Param3WriteEnabled=true
   ) {module=DynamicMaterialParameters001}
  ]
  rendererClasses [&NiagaraSpriteRendererProperties]
  renderers [{rendererClass=&NiagaraSpriteRendererProperties}]
 }
 Empty {
  emitterUpdateScript [
   &EmitterState()
   &SpawnBurst_Instantaneous(
    Spawn Count=&RandomRangeInt(Minimum=50, Maximum=100)
    Use Spawn Probability=false
    UseLoopCountLimit=false
   )
  ]
  particleSpawnScript [
   &InitializeParticle(
    Color=<rgba 100 1.5479 0.5 1>
    Lifetime=&RandomRangeFloat(Minimum=0.25, Maximum=&RandomRangeFloat(Minimum=0.75, Maximum=1.2))
    Position Offset=<0 0 120>
    Sprite Size Mode=::NewEnumerator1
    Uniform Sprite Size Max=2
    Uniform Sprite Size Min=1
    UsePositionOffset=true
    Write Material Random=true
   )
   &AddVelocity(
    Module.Use Falloff Distance=false
    Origin Offset=<0 0 120>
    Override Seed=false
    Velocity Mode=::NewEnumerator1
    Velocity Speed=&RandomRangeFloat(Minimum=75, Maximum=200)
   )
   &ApplyInitialForces()
   &ShapeLocation(MaskSpawnGroup=false, Override Seed=false, Sphere Radius=10, UseOwnerScale=true)
  ]
  particleUpdateScript [
   &ParticleState()
   &ScaleSpriteSize(Uniform Curve Sprite Scale=DataInterface() {propertyValues {Curve {keys [{interpMode=RCIM_Cubic, tangentMode=RCTM_User, time=0, value=1}, {interpMode=RCIM_Cubic, tangentMode=RCTM_User, time=1, value=0}]}}})
   &CurlNoiseForce(Noise Strength=500, Pan Noise=false)
   &GravityForce(Gravity=&RandomRangeVector(Minimum=<0 0 -100>, Maximum=<0 0 100>))
   &Drag(Use Linear Drag=true, Use Rotational Drag=true)
   &SolveForcesAndVelocity(Clamp Velocity=false, Limit Acceleration=false)
  ]
  rendererClasses [&NiagaraSpriteRendererProperties]
  renderers [{rendererClass=&NiagaraSpriteRendererProperties}]
 }
 Falling_Leaf {
  emitterUpdateScript [
   &EmitterState(
    Inactive Response=::NewEnumerator1
    Life Cycle Mode=::NewEnumerator1
    Loop Behavior=::NewEnumerator1
    Loop Duration Mode=::NewEnumerator1
    UseLoopDelay=false
   )
   &SpawnBurst_Instantaneous(
    Spawn Count=&MakeCustomIntFromBool(
     Bool=true
     True INT=&RandomRangeInt(Minimum=15, Maximum=25)
     False INT=&RandomRangeInt(Minimum=25, Maximum=30)
    )
    Use Spawn Probability=false
    UseLoopCountLimit=false
   )
   /SetVariables(Emitter.Hot Edge Color:LinearColor=&MakeCustomLinearColorFromBool(Bool=$User.IsFrontEnd, True Color=<rgba 6 0.5104 0.246 1>, False Color=<rgba 25 7.9348 0.9904 1>)) {isSetParameters=true}
  ]
  particleSpawnScript [
   /SetVariables(Particles.Variation:NiagaraFloat=&RandomRangeFloat(Minimum=0, Maximum=1)) {isSetParameters=true}
   &InitializeParticle(
    Color=&SelectLinearColorFromArray(
     Array Sampling Mode=::NewEnumerator0
     Direct Array Index=$Particles.PartitionParticles.Partition
     Color Selection Array=DataInterface() {propertyValues {ColorData [<rgba 0.2344 0.0248 0 1>, <rgba 0.5781 0.2192 0.0312 1>, <rgba 0.4401 0.3318 0.0722 1>, <rgba 0.0649 0.1849 0.0058 1>]}}
     Random Seed=0
    )
    Lifetime Min=0.5
    Lifetime Mode=::NewEnumerator1
    Mesh Scale Mode=::NewEnumerator3
    Mesh Uniform Scale=&Lerp_Float(A=0.2, B=0.4, Alpha=$Particles.Variation)
    UsePositionOffset=false
    Write Material Random=true
   )
   &InitialMeshOrientation(
    Mesh Orientation Mode=::NewEnumerator1
    Rotation=&RandomRangeVector(Minimum=<0 0 0>, Maximum=<1 1 1>)
    Use Rotation Vector=true
   )
   &SkeletalMeshLocation(
    Bone / Socket Index=&SelectIntFromArray(
     Array Sampling Mode=::NewEnumerator1
     Direct Array Index=$Particles.PartitionParticles.Partition
     Int Selection Array=DataInterface() {propertyValues {IntData [0,1,2,3,4,5,6]}}
     Random Seed=0
    )
    Bone Sampling Mode=::NewEnumerator4
    MaskSpawnGroup=false
    Orientation Sampling=::NewEnumerator1
    Position Sampling=::NewEnumerator1
    Sampled Position Offset=<0 0 20>
    Sampled Position Offset Coordinate Space:ENiagaraSkelMeshTransforms=::NewEnumerator1
    Skeletal Mesh=$System.Skeletal Mesh
   )
   &InheritVelocity(Speed Limit=true, UseSpeedThreshold=false)
   /SetVariables(
    Particles.MeshIndex=&RandomRangeInt(Minimum=0, Maximum=1)
    Particles.VisibilityTag=&RandomRangeInt(Minimum=0, Maximum=1)
   ) {isSetParameters=true}
   &AddVelocity(Velocity=&MakeVector(
     X=&RandomRangeFloat(Minimum=-200, Maximum=200)
     Y=&RandomRangeFloat(Minimum=-200, Maximum=200)
     Z=&RandomRangeFloat(Minimum=-50, Maximum=200)
    )) {module=AddVelocity001}
   /SetVariables(Particles.OakLeafSize:Vector3f=&Multiply_Vector(A=<0.5 0.7 1>, B=<1 1.5 1>)) {isSetParameters=true}
  ]
  particleUpdateScript [
   &CameraOffset(Camera Offset Amount=-25)
   &ParticleState()
   &ScaleMeshSize(Scale Factor=&VectorFromFloat(Value=&FloatFromCurve(FloatCurve=DataInterface() {propertyValues {Curve {keys [
          {arriveTangent=8.8695, interpMode=RCIM_Cubic, leaveTangent=8.8695, tangentMode=RCTM_User, time=0, value=0}
          {arriveTangent=-2.5249, leaveTangent=-2.5249, time=0.1408, value=0.9847}
          {arriveTangent=-1.2963, interpMode=RCIM_Cubic, leaveTangent=-1.2963, tangentMode=RCTM_User, time=0.5, value=1}
         ]}}})))
   &CurlNoiseForce(Noise Strength=50, Pan Noise=false)
   &GravityForce(Gravity=&RandomRangeVector(Minimum=<0 0 -350>, Maximum=<1 1 -200>))
   &Drag(
    Drag=&FloatFromCurve(
     FloatCurve=DataInterface() {propertyValues {Curve {keys [
         {arriveTangent=1.539, interpMode=RCIM_Cubic, leaveTangent=1.539, tangentMode=RCTM_User, time=0, value=0}
         {arriveTangent=1.539, interpMode=RCIM_Cubic, leaveTangent=1.539, tangentMode=RCTM_User, time=1, value=1}
        ]}}}
     Scale Curve=12
    )
    Use Linear Drag=true
    Use Rotational Drag=false
   ) {module=Drag001}
   &UpdateMeshOrientation(Rotation Rate=0.5, Rotation Vector=&RandomRangeVector(Minimum=<0 0 0>, Maximum=<1 1 1>))
   &SolveForcesAndVelocity(Clamp Velocity=false, Limit Acceleration=false)
   /SetVariables(Particles.RemainingLife:NiagaraFloat=&Subtract_Float(A=$Particles.Lifetime, B=$Particles.Age)) {isSetParameters=true}
   /SetVariables(Particles.MeshScale1:Vector3f=&Add_Vector(A=$StackContext.Scale, B=<0.1 0.1 0.1>)) {isSetParameters=true}
   &ScaleMeshSize(Scale Factor=&VectorFromFloat(Value=&Lerp_Float(
      A=1
      B=2
      Alpha=&FloatFromCurve(
       FloatCurve=DataInterface() {propertyValues {Curve {keys [
           {arriveTangent=9.0725, interpMode=RCIM_Cubic, time=0, value=1}
           {arriveTangent=-0.3899, time=0.1, value=0}
           {interpMode=RCIM_Cubic, leaveTangent=1.0526, time=1, value=0}
          ]}}}
       CurveIndex=$NPC.NPC_Sparks_Audio_Parameters001.BeatLifetime
      )
     ))) {enabled=false, module=ScaleMeshSize001}
   &ScaleMeshSize(Scale Factor=&VectorFromFloat(Value=&Lerp_Float(
      A=1
      B=2.15
      Alpha=&FloatFromCurve(
       FloatCurve=DataInterface() {propertyValues {Curve {keys [
           {arriveTangent=9.0725, interpMode=RCIM_Cubic, time=0, value=1}
           {arriveTangent=-0.3899, time=0.15, value=0}
           {arriveTangent=0.159, interpMode=RCIM_Cubic, leaveTangent=0.159, tangentMode=RCTM_User, time=0.975, value=0}
           {interpMode=RCIM_Cubic, leaveTangent=1.0526, time=1, value=1}
          ]}}}
       CurveIndex=$NPC.NPC_Sparks_Audio_Parameters001.MeasureLifetime
      )
     ))) {enabled=false, module=ScaleMeshSize002}
   &DynamicMaterialParameters(
    Index 0 Param 1=&RandomRangeFloat(Minimum=0, Maximum=1)
    Index 0 Param 2=&RandomRangeFloat(Minimum=0, Maximum=0.4)
    Index 0 Param 3=&FloatFromCurve(FloatCurve=DataInterface() {propertyValues {Curve {keys [
         <at 0 1 RCIM_Cubic>
         {arriveTangent=0.0122, interpMode=RCIM_Cubic, leaveTangent=-0.0131, tangentMode=RCTM_Break, time=0.2, value=1}
         {interpMode=RCIM_Cubic, tangentMode=RCTM_User, time=0.5, value=0.86}
         {arriveTangent=-1.72, interpMode=RCIM_Cubic, leaveTangent=-1.72, time=0.9499, value=0.3175}
         {arriveTangent=-9.8557, interpMode=RCIM_Cubic, tangentMode=RCTM_Break, time=1, value=0}
        ]}}})
    Index 0 Param 4=&FloatFromCurve(
     FloatCurve=DataInterface() {propertyValues {Curve {keys [
         {interpMode=RCIM_Cubic, tangentMode=RCTM_User, time=0.5, value=0.05}
         {arriveTangent=2.6834, interpMode=RCIM_Cubic, leaveTangent=2.6834, tangentMode=RCTM_User, time=1, value=0.8}
        ]}}}
     Scale Curve=&Lerp_Float(A=1.5, B=0.5, Alpha=$Particles.Variation)
    )
    Index 1 Param 1=&RandomRangeFloat(Minimum=0, Maximum=3)
    Index 1 Param 2=&RandomRangeFloat(Minimum=0, Maximum=8.9935)
    Index 1 Param 3=&Lerp_Float(
     A=&RandomRangeFloat(Minimum=0, Maximum=0.48)
     B=&RandomRangeFloat(Minimum=0, Maximum=5.1449)
     Alpha=&MakeFloatFromInt(Int=$Particles.VisibilityTag)
    )
    Module.Index0Param0Write=false
    Module.Index0Param1Write=false
    Module.Index0Param2Write=true
    Module.Index0Param3Write=true
    Module.Index1Param0Write=true
    Module.Index1Param1Write=true
    Module.Index1Param2Write=true
    Module.Index1Param3Write=true
    Write Parameter Index 1=true
   )
   &DynamicMaterialParameters(
    Index 0 Param 3=0.356
    Index 0 Param 4=0.05
    Module.Index0Param0Write=false
    Module.Index0Param1Write=false
    Module.Index0Param2Write=true
    Module.Index0Param3Write=true
   ) {enabled=false, module=DynamicMaterialParameters001}
   /SetVariables(Particles.OakLeafSize:Vector3f=&Multiply_VectorByFloat(
     Vector=$Particles.Initial.OakLeafSize
     Float=&FloatFromCurve(FloatCurve=DataInterface() {propertyValues {Curve {keys [
          {arriveTangent=8.8695, interpMode=RCIM_Cubic, leaveTangent=8.8695, tangentMode=RCTM_User, time=0, value=0}
          {arriveTangent=-2.5249, leaveTangent=-2.5249, time=0.1408, value=0.9847}
          {arriveTangent=-1.2963, interpMode=RCIM_Cubic, leaveTangent=-1.2963, tangentMode=RCTM_User, time=0.5, value=1}
         ]}}})
    )) {isSetParameters=true}
  ]
  rendererClasses [&NiagaraMeshRendererProperties]
  renderers [
   {rendererClass=&NiagaraMeshRendererProperties, rendererIndex=0}
   {rendererClass=&NiagaraMeshRendererProperties, rendererIndex=1}
  ]
 }
 Flare {
  emitterUpdateScript [
   &EmitterState(Life Cycle Mode=::NewEnumerator1, Loop Behavior=::NewEnumerator1, Loop Duration=1, UseLoopDelay=false)
   &SpawnBurst_Instantaneous(Spawn Count=1, Use Spawn Probability=false)
  ]
  particleSpawnScript [
   &InitializeParticle(
    Color=<rgba 2 0.4833 0.143 1>
    Lifetime=0.3
    Sprite Rotation Mode=::NewEnumerator2
    Sprite Size Mode=::NewEnumerator3
    Uniform Sprite Size=600
    UsePositionOffset=false
    Write Material Random=true
   )
   &DynamicMaterialParameters(
    Index 0 Param 1=&RandomRangeFloat(Minimum=0.4, Maximum=0.6)
    Index 0 Param 2=1
    Index 0 Param 3=1
    Index 0 Param 4=1
    Module.Param0WriteEnabled=true
    Module.Param1WriteEnabled=true
    Module.Param2WriteEnabled=true
    Module.Param3WriteEnabled=true
   )
   &SystemLocation(MaskSpawnGroup=false, Offset=<0 0 120>)
   /SetVariables(Particles.Sharp Flare Camera Offset=40) {isSetParameters=true}
   &CameraOffset(Camera Offset Amount=-30) {module=CameraOffset001}
  ]
  particleUpdateScript [
   &ParticleState()
   &SolveForcesAndVelocity(Clamp Velocity=false, Limit Acceleration=false)
   &SkeletalMeshLocation(Bone / Socket Index=7, Bone Sampling Mode=::NewEnumerator4, MaskSpawnGroup=false, Skeletal Mesh=$System.Skeletal Mesh) {module=SkeletalMeshLocation001}
   &ScaleColor(
    Scale Alpha=&Multiply_Float(
     A=&FNTimeOfDayScalar(Night Scale=1, Morning Scale=1, Evening Scale=1)
     B=&FloatFromCurve(FloatCurve=DataInterface() {propertyValues {Curve {keys [<at 0 0>,<at 0.2 1>,<at 0.8 1>,<at 1 0>]}}})
    )
    Scale RGB=&Multiply_VectorByFloat(Vector=<1 1 1>, Float=1)
    ScaleA=true
    ScaleRGB=true
   )
   &ScaleSpriteSize(Scale Factor=&Vector2DFromFloat(Value=&FloatFromCurve(FloatCurve=DataInterface() {propertyValues {Curve {keys [<at 0 1.5 RCIM_Cubic>,<at 1 0>]}}})))
   /SetVariables(Particles.Flash color:LinearColor=&MakeLinearColorFromVectorAndFloat(
     "Vector (RGB)"=&MakeVectorFromLinearColorRGB(Color=<rgba 1.25 0.297 0.0906 1>)
     "Float (Alpha)"=&Multiply_Float(
      A=&Multiply_Float(
       A=&FNTimeOfDayScalar(Night Scale=1, Morning Scale=1, Evening Scale=1)
       B=&FloatFromCurve(FloatCurve=DataInterface() {propertyValues {Curve {keys [<at 0 0>,<at 0.124 1>,<at 0.496 1>,<at 0.62 0>]}}})
      )
      B=1
     )
    )) {isSetParameters=true}
   /SetVariables(Particles.Flash size:Vector2f=&Multiply_Vector2D_ByFloat(
     Float=&Multiply_Float(
      A=400
      B=&FloatFromCurve(FloatCurve=DataInterface() {propertyValues {Curve {keys [
           {arriveTangent=4.1785, interpMode=RCIM_Cubic, leaveTangent=4.1785, tangentMode=RCTM_User, time=0, value=0}
           {arriveTangent=-0.15, interpMode=RCIM_Cubic, leaveTangent=-0.15, tangentMode=RCTM_User, time=0.1378, value=1}
           {arriveTangent=-0.0405, interpMode=RCIM_Cubic, leaveTangent=-0.0405, tangentMode=RCTM_User, time=0.62, value=0.25}
          ]}}})
     )
     Vector2D=<1 1>
    )) {isSetParameters=true}
  ]
  rendererClasses [&NiagaraSpriteRendererProperties]
  renderers [
   {rendererClass=&NiagaraSpriteRendererProperties}
   {rendererClass=&NiagaraSpriteRendererProperties, rendererIndex=1}
  ]
 }
 LeafEmbers {
  emitterUpdateScript [
   &EmitterState(
    Inactive Response=::NewEnumerator1
    Life Cycle Mode=::NewEnumerator1
    Loop Behavior=::NewEnumerator1
    Loop Duration Mode=::NewEnumerator1
    UseLoopDelay=false
   )
   &SpawnParticlesFromOtherEmitter(
    Attribute Reader=DataInterface() {propertyValues {EmitterBinding {bindingMode=Other, emitterName=Falling_Leaf}}}
    Spawn Rate=6
    Use Maximum Age=false
    Use Minimum Age=false
   )
  ]
  particleSpawnScript [
   &InitializeParticle(
    Color=<rgba 100 1.5479 0.5 1>
    Lifetime Max=1
    Lifetime Min=0.5
    Lifetime Mode=::NewEnumerator1
    Mesh Scale Mode=::NewEnumerator3
    Mesh Uniform Scale=&RandomRangeFloat(Minimum=0.05, Maximum=0.25)
    Sprite Rotation Mode=::NewEnumerator2
    Sprite Size Mode=::NewEnumerator1
    Uniform Sprite Size Max=1.3
    Uniform Sprite Size Min=0.75
    UsePositionOffset=false
    Write Material Random=true
   )
   &SampleParticlesFromOtherEmitter(Age Sampling:ENiagara_DirectReadSamplingMode=::NewEnumerator1, Lifetime Sampling:ENiagara_DirectReadSamplingMode=::NewEnumerator1)
   /SetVariables(Transient.ParentNormalizedAge:NiagaraFloat=&Divide_Float(A=$Output.SampleParticlesFromOtherEmitter.SampledAge, B=$Output.SampleParticlesFromOtherEmitter.SampledLifetime)) {isSetParameters=true}
   &ShapeLocation(
    MaskSpawnGroup=false
    Override Seed=false
    Sphere Radius=&FloatFromCurve(FloatCurve=DataInterface() {propertyValues {Curve {keys [<at 0.33 1>,<at 1 0>]}}}, CurveIndex=$Transient.ParentNormalizedAge, Scale Curve=5)
    UseOwnerScale=true
   )
   &InheritVelocity(Speed Limit=true, UseSpeedThreshold=false)
   &AddVelocity(Velocity=&RandomRangeVector(Minimum=<5 5 15>, Maximum=<-5 -5 10>), Velocity Speed Scale=3)
   &KillParticles(Kill Particles=&SetBoolByFloatComparison(
     A=&FloatFromCurve(
      FloatCurve=DataInterface() {propertyValues {Curve {keys [
          {arriveTangent=25.7884, interpMode=RCIM_Cubic, time=0.33, value=0}
          {arriveTangent=-0.0504, interpMode=RCIM_Cubic, leaveTangent=-0.0504, tangentMode=RCTM_User, time=0.4, value=1}
          {arriveTangent=0.0831, interpMode=RCIM_Cubic, leaveTangent=0.0831, tangentMode=RCTM_User, time=0.6, value=1}
          {arriveTangent=-10.9519, interpMode=RCIM_Constant, tangentMode=RCTM_User, time=0.8001, value=0}
         ]}}}
      CurveIndex=$Transient.ParentNormalizedAge
     )
     B=&RandomRangeFloat(Minimum=0, Maximum=1)
     Comparison Type=::NewEnumerator6
    )) {module=KillParticles001}
  ]
  particleUpdateScript [
   &CameraOffset(Camera Offset Amount=-25)
   &ParticleState()
   &ScaleSpriteSize(Uniform Curve Sprite Scale=DataInterface() {propertyValues {Curve {keys [{interpMode=RCIM_Cubic, tangentMode=RCTM_User, time=0, value=1}, {interpMode=RCIM_Cubic, tangentMode=RCTM_User, time=1, value=0}]}}})
   &Drag(Drag=0.15, Rotational Drag=0.1, Use Linear Drag=true, Use Rotational Drag=true)
   &SolveForcesAndVelocity(Clamp Velocity=false, Limit Acceleration=false)
  ]
  rendererClasses [&NiagaraSpriteRendererProperties]
  renderers [
   {rendererClass=&NiagaraSpriteRendererProperties}
   {rendererClass=&NiagaraSpriteRendererProperties, rendererIndex=1}
  ]
 }
}
summary {
 systemName=NS_Aura_Fall_Start
 emitters [
  {emitterName=Corona, rendererClasses [&NiagaraSpriteRendererProperties]}
  {emitterName=Flare, rendererClasses [&NiagaraSpriteRendererProperties]}
  {emitterName=LeafEmbers, rendererClasses [&NiagaraSpriteRendererProperties]}
  {emitterName=Falling_Leaf, rendererClasses [&NiagaraMeshRendererProperties]}
  {emitterName=Empty, rendererClasses [&NiagaraSpriteRendererProperties]}
 ]
 userVariables [
  User.Interpolate Intensity {type=NiagaraFloat, defaultValue=0.5, description=""}
  User.IsFrontEnd {type=NiagaraBool}
  User.IsLargeBody {type=NiagaraBool}
 ]
}
