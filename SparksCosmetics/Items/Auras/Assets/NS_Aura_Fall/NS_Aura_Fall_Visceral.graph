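# Niagara autumn aura cosmetic that surrounds a character with falling leaves, rising embers and a warm glow placed on skeletal mesh bones or sockets.
# Mesh leaves drift on curl noise and feed LeafEmbers through an attribute reader, with leaf age shaping ember radius and culling.
# Seven looping glow sprites follow the body, scaled by time of day.
# User flags adjust for front-end and large bodies; the interpolated shard glow emitter is disabled.
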NiagaraSystem /SparksCosmetics/Items/Auras/Assets/NS_Aura_Fall/NS_Aura_Fall_Visceral
dependencies {
 usedDataInterfaces [
  &NiagaraDataInterfaceArrayFloat
  &NiagaraDataInterfaceArrayInt32
  &NiagaraDataInterfaceArrayColor
  &NiagaraDataInterfaceCurve
  &NiagaraDataInterfaceParticleRead
  &NiagaraDataInterfaceSkeletalMesh
 ]
 usedDynamicInputs []
 usedModules [
  &self:SystemSpawnScript.SystemScriptSource.SystemScriptGraph.NiagaraNodeAssignment_0.SetVariables_#1
  &self:SystemSpawnScript.SystemScriptSource.SystemScriptGraph.NiagaraNodeAssignment_1.SetVariables_#2
  &SystemState
 ]
}
emitters {
 Body_Embers {
  emitterUpdateScript [
   &EmitterState(
    Inactive Response=::NewEnumerator1
    Life Cycle Mode=::NewEnumerator1
    Loop Behavior=::NewEnumerator1
    Loop Duration Mode=::NewEnumerator1
    UseLoopDelay=false
   )
   &SpawnRate(SpawnRate=5, Use Spawn Probability=false)
  ]
  particleSpawnScript [
   &PartitionParticles(Number Of Partitions=7, Particles Per Partition=1, Partition Mode:ENiagara_PartitionMode=::NewEnumerator1)
   &InitializeParticle(
    Color Maximum=<rgba 100 69.0789 0.3 1>
    Color Minimum=<rgba 100 0.4 0.4 1>
    Color Mode=::NewEnumerator2
    Lifetime Max=1.5
    Lifetime Mode=::NewEnumerator1
    Mesh Scale Mode=::NewEnumerator3
    Mesh Uniform Scale=&RandomRangeFloat(Minimum=0.05, Maximum=0.25)
    Sprite Rotation Mode=::NewEnumerator0
    Sprite Size Mode=::NewEnumerator1
    Uniform Sprite Size Max=3
    Uniform Sprite Size Min=1
    UsePositionOffset=false
    Write Material Random=true
   )
   &SkeletalMeshLocation(
    Bone / Socket Index=&SelectIntFromArray(Direct Array Index=$Particles.PartitionParticles.Partition, Int Selection Array=DataInterface() {propertyValues {IntData [0,1,2,3,4,5,6]}})
    Bone Sampling Mode=::NewEnumerator4
    MaskSpawnGroup=false
    Orientation Sampling=::NewEnumerator1
    Position Sampling=::NewEnumerator1
    Sampled Position Offset Coordinate Space:ENiagaraSkelMeshTransforms=::NewEnumerator1
    Skeletal Mesh=$System.Skeletal Mesh
   )
   &InheritVelocity(Speed Limit=true, UseSpeedThreshold=false)
   /SetVariables(Particles.MeshIndex=&SelectIntFromArray(Direct Array Index=$Particles.PartitionParticles.Partition, Int Selection Array=DataInterface() {propertyValues {IntData [0,1,1,2,2,1,1]}})) {isSetParameters=true, module=SetVariables_#3}
   &AddVelocity(Velocity=&RandomRangeVector(Minimum=<5 5 15>, Maximum=<-5 -5 10>), Velocity Speed Scale=3)
  ]
  particleUpdateScript [
   &CameraOffset(Camera Offset Amount=-25)
   &ParticleState()
   &ScaleSpriteSize(Uniform Curve Sprite Scale=DataInterface() {propertyValues {Curve {keys [{interpMode=RCIM_Cubic, tangentMode=RCTM_User, time=0, value=1}, {interpMode=RCIM_Cubic, tangentMode=RCTM_User, time=1, value=0}]}}})
   &UpdateMeshOrientation(Rotation Rate=0.5, Rotation Vector=&RandomRangeVector(Minimum=<0 0 0>, Maximum=<1 1 1>))
   &UpdateMeshOrientation(
    Facing Direction=/Game/Effects/Niagara/DynamicInputs/Engine/FN_GetSystemAxis(Axis:ENiagaraOrientationAxis=::XAxis, Flip Axis:NiagaraBool=false)
    Facing Interpolation=false
    Orientation Method=::NewEnumerator1
    Override Quat=false
    TransformFacing=false
    TransformUp=false
   ) {enabled=false, module=UpdateMeshOrientation001}
   &ScaleMeshSize(Scale Factor=&VectorFromFloat(Value=&FloatFromCurve(FloatCurve=DataInterface() {propertyValues {Curve {keys [
          {arriveTangent=9.0725, interpMode=RCIM_Cubic, leaveTangent=9.0725, tangentMode=RCTM_User, time=0, value=0}
          {arriveTangent=0.1393, leaveTangent=0.1393, time=0.05, value=0.9847}
          {arriveTangent=-1.0365, interpMode=RCIM_Cubic, leaveTangent=-1.0365, time=0.5, value=1}
          {arriveTangent=-0.0405, interpMode=RCIM_Cubic, leaveTangent=-0.0405, tangentMode=RCTM_User, time=1, value=0}
         ]}}})))
   &Drag(Drag=0.15, Rotational Drag=0.1, Use Linear Drag=true, Use Rotational Drag=true)
   &SolveForcesAndVelocity(Clamp Velocity=false, Limit Acceleration=false)
  ]
  rendererClasses [&NiagaraSpriteRendererProperties]
  renderers [{rendererClass=&NiagaraSpriteRendererProperties}]
 }
 Falling_Leaf {
  emitterUpdateScript [
   &EmitterState(
    Inactive Response=::NewEnumerator1
    Life Cycle Mode=::NewEnumerator1
    Loop Behavior=::NewEnumerator1
    Loop Duration Mode=::NewEnumerator1
    UseLoopDelay=false
   )
   &SpawnRate(SpawnRate=5, Use Spawn Probability=false)
   /SetVariables(Emitter.Hot Edge Color:LinearColor=&MakeCustomLinearColorFromBool(Bool=$User.IsFrontEnd, True Color=<rgba 6 0.5104 0.246 1>, False Color=<rgba 25 7.9348 0.9904 1>)) {isSetParameters=true}
  ]
  particleSpawnScript [
   /SetVariables(Particles.Variation:NiagaraFloat=&RandomRangeFloat(Minimum=0, Maximum=1)) {isSetParameters=true}
   /SetVariables(
    Particles.MeshIndex=&RandomRangeInt(Minimum=0, Maximum=1)
    Particles.VisibilityTag=&RandomRangeInt(Minimum=0, Maximum=1)
   ) {isSetParameters=true, module=SetVariables_#3}
   &InitializeParticle(
    Color=&MakeCustomLinearColorFromBool(
     Bool=$User.IsFrontEnd
     True Color=&SelectLinearColorFromArray(
      Array Sampling Mode=::NewEnumerator0
      Direct Array Index=?
      Color Selection Array=DataInterface() {propertyValues {ColorData [<rgba 0.7188 0.0759 0 1>, <rgba 0.7109 0.2444 0 1>, <rgba 0.5521 0.4095 0.0051 1>, <rgba 0.4792 0.4679 0 1>]}}
      Random Seed=0
     )
     False Color=&SelectLinearColorFromArray(
      Array Sampling Mode=::NewEnumerator0
      Direct Array Index=?
      Color Selection Array=DataInterface() {propertyValues {ColorData [<rgba 0.2344 0.0248 0 1>, <rgba 0.7891 0.2767 0.0082 1>, <rgba 0.4401 0.3265 0.0041 1>, <rgba 0.1849 0.1806 0 1>]}}
      Random Seed=0
     )
    )
    Lifetime Max=3.2
    Lifetime Min=2.8
    Lifetime Mode=::NewEnumerator1
    Mesh Scale Mode=::NewEnumerator3
    Mesh Uniform Scale=&MakeCustomFloatFromBool(
     Bool=&SetBoolByIntComparison(A=$Particles.VisibilityTag, B=0, Comparison Type=::NewEnumerator0)
     True Float=&Lerp_Float(A=0.3, B=0.5, Alpha=$Particles.Variation)
     False Float=&Lerp_Float(A=0.2, B=0.4, Alpha=$Particles.Variation)
    )
    UsePositionOffset=false
    Write Material Random=true
   )
   &InitialMeshOrientation(
    Mesh Orientation Mode=::NewEnumerator1
    Rotation=&RandomRangeVector(Minimum=<0 0 0>, Maximum=<1 1 1>)
    Use Rotation Vector=true
   )
   &SkeletalMeshLocation(
    Bone / Socket Index=&SelectIntFromArray(
     Array Sampling Mode=::NewEnumerator1
     Direct Array Index=$Particles.PartitionParticles.Partition
     Int Selection Array=DataInterface() {propertyValues {IntData [0,1,2,3,4,5,6]}}
     Random Seed=0
    )
    Bone Sampling Mode=::NewEnumerator4
    MaskSpawnGroup=false
    Orientation Sampling=::NewEnumerator1
    Position Sampling=::NewEnumerator1
    Sampled Position Offset=<0 0 55>
    Sampled Position Offset Coordinate Space:ENiagaraSkelMeshTransforms=::NewEnumerator1
    Skeletal Mesh=$System.Skeletal Mesh
   )
   &InheritVelocity(Speed Limit=true, UseSpeedThreshold=false)
   &AddVelocity(Velocity=&RandomRangeVector(Minimum=<5 5 -20>, Maximum=<-5 -5 -8>), Velocity Speed Scale=3)
   /SetVariables(Particles.OakLeafSize:Vector3f=&Multiply_Vector(A=<0.5 0.7 1>, B=<1 1.5 1>)) {isSetParameters=true}
  ]
  particleUpdateScript [
   &CameraOffset(Camera Offset Amount=-25)
   &ParticleState()
   &UpdateMeshOrientation(Rotation Rate=0.5, Rotation Vector=&RandomRangeVector(Minimum=<0 0 0>, Maximum=<1 1 1>))
   &ScaleMeshSize(Scale Factor=&VectorFromFloat(Value=&FloatFromCurve(FloatCurve=DataInterface() {propertyValues {Curve {keys [
          {arriveTangent=0.1318, interpMode=RCIM_Cubic, leaveTangent=0.1318, tangentMode=RCTM_User, time=0, value=0}
          {arriveTangent=0.0054, interpMode=RCIM_Cubic, leaveTangent=0.0445, tangentMode=RCTM_Break, time=0.3, value=1}
          {arriveTangent=-0.0634, interpMode=RCIM_Cubic, leaveTangent=-1.6812, tangentMode=RCTM_Break, time=0.9, value=1}
          <at 1 0 RCIM_Cubic>
         ]}}})))
   &Drag(Drag=0.15, Rotational Drag=0.1, Use Linear Drag=true, Use Rotational Drag=true)
   &CurlNoiseForce(Noise Strength=50, Pan Noise=false)
   &SolveForcesAndVelocity(Clamp Velocity=false, Limit Acceleration=false)
   /SetVariables(Particles.RemainingLife:NiagaraFloat=&Subtract_Float(A=$Particles.Lifetime, B=$Particles.Age)) {isSetParameters=true}
   /SetVariables(Particles.MeshScale1:Vector3f=&Add_Vector(A=$StackContext.Scale, B=<0.1 0.1 0.1>)) {isSetParameters=true}
   &ScaleMeshSize(Scale Factor=&VectorFromFloat(Value=&Lerp_Float(
      A=1
      B=2
      Alpha=&FloatFromCurve(
       FloatCurve=DataInterface() {propertyValues {Curve {keys [
           {arriveTangent=9.0725, interpMode=RCIM_Cubic, time=0, value=1}
           {arriveTangent=-0.3899, time=0.1, value=0}
           {interpMode=RCIM_Cubic, leaveTangent=1.0526, time=1, value=0}
          ]}}}
       CurveIndex=$NPC.NPC_Sparks_Audio_Parameters001.BeatLifetime
      )
     ))) {enabled=false, module=ScaleMeshSize001}
   &ScaleMeshSize(Scale Factor=&VectorFromFloat(Value=&Lerp_Float(
      A=1
      B=2.15
      Alpha=&FloatFromCurve(
       FloatCurve=DataInterface() {propertyValues {Curve {keys [
           {arriveTangent=9.0725, interpMode=RCIM_Cubic, time=0, value=1}
           {arriveTangent=-0.3899, time=0.15, value=0}
           {arriveTangent=0.159, interpMode=RCIM_Cubic, leaveTangent=0.159, tangentMode=RCTM_User, time=0.975, value=0}
           {interpMode=RCIM_Cubic, leaveTangent=1.0526, time=1, value=1}
          ]}}}
       CurveIndex=$NPC.NPC_Sparks_Audio_Parameters001.MeasureLifetime
      )
     ))) {enabled=false, module=ScaleMeshSize002}
   &DynamicMaterialParameters(
    Index 0 Param 1=&RandomRangeFloat(Minimum=0, Maximum=1)
    Index 0 Param 2=&RandomRangeFloat(Minimum=0, Maximum=0.4)
    Index 0 Param 3=&FloatFromCurve(FloatCurve=DataInterface() {propertyValues {Curve {keys [
         <at 0 1 RCIM_Cubic>
         {arriveTangent=0.0122, interpMode=RCIM_Cubic, leaveTangent=-0.0131, tangentMode=RCTM_Break, time=0.2, value=1}
         {interpMode=RCIM_Cubic, tangentMode=RCTM_User, time=0.5011, value=0.9}
         {arriveTangent=-1.8039, interpMode=RCIM_Cubic, leaveTangent=-1.8039, time=0.9499, value=0.3175}
         {arriveTangent=-9.8557, interpMode=RCIM_Cubic, tangentMode=RCTM_Break, time=1, value=0}
        ]}}})
    Index 0 Param 4=&FloatFromCurve(
     FloatCurve=DataInterface() {propertyValues {Curve {keys [
         {interpMode=RCIM_Cubic, tangentMode=RCTM_User, time=0.5, value=0.05}
         {arriveTangent=2.6834, interpMode=RCIM_Cubic, leaveTangent=2.6834, tangentMode=RCTM_User, time=1, value=0.8}
        ]}}}
     Scale Curve=&Lerp_Float(A=1.5, B=0.5, Alpha=$Particles.Variation)
    )
    Index 1 Param 1=&RandomRangeFloat(Minimum=0, Maximum=1)
    Index 1 Param 2=&RandomRangeFloat(Minimum=0, Maximum=3)
    Index 1 Param 3=&Lerp_Float(
     A=&RandomRangeFloat(Minimum=0, Maximum=0.48)
     B=&RandomRangeFloat(Minimum=0, Maximum=0.8298)
     Alpha=&MakeFloatFromInt(Int=$Particles.VisibilityTag)
    )
    Module.Index0Param0Write=false
    Module.Index0Param1Write=false
    Module.Index0Param2Write=true
    Module.Index0Param3Write=true
    Module.Index1Param0Write=true
    Module.Index1Param1Write=true
    Module.Index1Param2Write=true
    Module.Index1Param3Write=true
    Write Parameter Index 1=true
   ) {module=DynamicMaterialParameters001}
   &Color(Color=<rgba 0.0062 0 1 1>) {enabled=false}
   /SetVariables(Particles.OakLeafSize:Vector3f=&Multiply_VectorByFloat(
     Vector=$Particles.Initial.OakLeafSize
     Float=&FloatFromCurve(FloatCurve=DataInterface() {propertyValues {Curve {keys [
          {arriveTangent=0.1318, interpMode=RCIM_Cubic, leaveTangent=0.1318, tangentMode=RCTM_User, time=0, value=0}
          {arriveTangent=0.0054, interpMode=RCIM_Cubic, leaveTangent=0.0445, tangentMode=RCTM_Break, time=0.3, value=1}
          {arriveTangent=-0.0634, interpMode=RCIM_Cubic, leaveTangent=-1.6812, tangentMode=RCTM_Break, time=0.9, value=1}
          <at 1 0 RCIM_Cubic>
         ]}}})
    )) {isSetParameters=true}
  ]
  rendererClasses [&NiagaraMeshRendererProperties]
  renderers [
   {rendererClass=&NiagaraMeshRendererProperties, rendererIndex=0}
   {rendererClass=&NiagaraMeshRendererProperties, rendererIndex=1}
  ]
 }
 Glow {
  emitterUpdateScript [
   &EmitterState(
    Inactive Response=::NewEnumerator1
    Life Cycle Mode=::NewEnumerator1
    Loop Behavior=::NewEnumerator1
    Loop Duration Mode=::NewEnumerator1
    UseLoopDelay=false
   )
   &SpawnBurst_Instantaneous(Spawn Count=7, Use Spawn Probability=false, UseLoopCountLimit=false)
  ]
  particleSpawnScript [
   &PartitionParticles(Number Of Partitions=7)
   &InitializeParticle(
    Color=&MakeCustomLinearColorFromBool(
     Bool=$User.IsFrontEnd
     True Color=&RandomRangeLinearColor(Minimum=<rgba 2 1.0723 0.479 1>, Maximum=<rgba 2 1.134 0.3022 1>)
     False Color=&RandomRangeLinearColor(Minimum=<rgba 15 3.6618 0.5773 1>, Maximum=<rgba 15 5.7224 0.449 1>)
    )
    Lifetime=0.1
    Sprite Rotation Mode=::NewEnumerator0
    Sprite Size Mode=::NewEnumerator3
    Uniform Sprite Size=&Multiply_Float(
     A=&SelectFloatFromArray(
      Direct Array Index=&SpawnGroup()
      Float Selection Array=DataInterface() {propertyValues {FloatData [40,40,20,20,30,30]}}
     )
     B=&MakeCustomFloatFromBool(Bool=$User.IsLargeBody, True Float=1.5, False Float=1)
    )
    UsePositionOffset=false
    Write Material Random=true
   )
   /SetVariables(Particles.glowScale:Vector2f=&Vector2DFromFloat(Value=&Multiply_Float(
      A=&SelectFloatFromArray(
       Direct Array Index=$Particles.PartitionParticles.Partition
       Float Selection Array=DataInterface() {propertyValues {FloatData [175,80,100,90,80,70,70,90]}}
      )
      B=&MakeCustomFloatFromBool(Bool=$User.IsLargeBody, True Float=1.2, False Float=1)
     ))) {isSetParameters=true}
   /SetVariables(Particles.glowCameraOffset=-10) {isSetParameters=true}
   /SetVariables(Particles.glowColor:LinearColor=&MultiplyLinearColor(
     Linear Color=&MultiplyLinearColor(Linear Color=<rgba 3 0.428 0 1>, Scale Factor=&MakeVector4_FromFloat(Float=3))
     Scale Factor=&MakeVector4_FromFloat(Float=&FNTimeOfDayScalar(
       Night Scale=0.25
       Morning Scale=0.2
       Day Scale=&MakeCustomFloatFromBool(Bool=$User.IsFrontEnd, True Float=0.07, False Float=0.26)
       Evening Scale=0.15
      ))
    )) {isSetParameters=true}
  ]
  particleUpdateScript [
   &ParticleState(
    Kill Particles When Lifetime Has Elapsed=false
    Let Infinitely Lived Particles Die When Emitter Deactivates=true
    Loop Particles Lifetime=true
   )
   &CameraOffset(Camera Offset Amount=5)
   &SpriteRotationRate(Rotation Rate=&RandomRangeFloat(Minimum=10, Maximum=20))
   &SkeletalMeshLocation(
    Bone / Socket Index=&SelectIntFromArray(Direct Array Index=$Particles.PartitionParticles.Partition, Int Selection Array=DataInterface() {propertyValues {IntData [3,1,2,4,5,6,0]}})
    Bone Sampling Mode=::NewEnumerator4
    MaskSpawnGroup=false
    Orientation Sampling=::NewEnumerator1
    Sampled Position Offset Coordinate Space:ENiagaraSkelMeshTransforms=::NewEnumerator1
    Skeletal Mesh=$System.Skeletal Mesh
   ) {module=SkeletalMeshLocation003}
   &DynamicMaterialParameters(
    Index 0 Param 1=&SelectFloatFromArray(Direct Array Index=$Particles.PartitionParticles.Partition, Float Selection Array=DataInterface() {propertyValues {FloatData [5,5,5,5,3,3]}})
    Index 0 Param 3=1
    Module.Param0WriteEnabled=true
    Module.Param1WriteEnabled=true
    Module.Param2WriteEnabled=true
    Module.Param3WriteEnabled=true
   )
   &ScaleColor(
    Scale RGB=&Multiply_VectorByFloat(
     Vector=&VectorFromFloat(Value=&FNTimeOfDayScalar(
       Night Scale=0.25
       Morning Scale=0.2
       Day Scale=&MakeCustomFloatFromBool(Bool=$User.IsFrontEnd, True Float=0.2, False Float=0.26)
       Evening Scale=0.15
      ))
     Float=4
    )
    ScaleA=true
    ScaleRGB=true
   )
  ]
  rendererClasses [&NiagaraSpriteRendererProperties]
  renderers [
   {rendererClass=&NiagaraSpriteRendererProperties}
   {rendererClass=&NiagaraSpriteRendererProperties, rendererIndex=1}
  ]
 }
 LeafEmbers {
  emitterUpdateScript [
   &EmitterState(
    Inactive Response=::NewEnumerator1
    Life Cycle Mode=::NewEnumerator1
    Loop Behavior=::NewEnumerator1
    Loop Duration Mode=::NewEnumerator1
    UseLoopDelay=false
   )
   &SpawnParticlesFromOtherEmitter(
    Attribute Reader=DataInterface() {propertyValues {EmitterBinding {bindingMode=Other, emitterName=Falling_Leaf}}}
    Use Maximum Age=false
    Use Minimum Age=false
   )
  ]
  particleSpawnScript [
   &InitializeParticle(
    Color=<rgba 100 1.5479 0.5 1>
    Lifetime Max=1
    Lifetime Min=0.5
    Lifetime Mode=::NewEnumerator1
    Mesh Scale Mode=::NewEnumerator3
    Mesh Uniform Scale=&RandomRangeFloat(Minimum=0.05, Maximum=0.25)
    Sprite Rotation Mode=::NewEnumerator2
    Sprite Size Mode=::NewEnumerator1
    Uniform Sprite Size Max=1.3
    Uniform Sprite Size Min=0.75
    UsePositionOffset=false
    Write Material Random=true
   )
   &SampleParticlesFromOtherEmitter(Age Sampling:ENiagara_DirectReadSamplingMode=::NewEnumerator1, Lifetime Sampling:ENiagara_DirectReadSamplingMode=::NewEnumerator1)
   /SetVariables(Transient.ParentNormalizedAge:NiagaraFloat=&Divide_Float(A=$Output.SampleParticlesFromOtherEmitter.SampledAge, B=$Output.SampleParticlesFromOtherEmitter.SampledLifetime)) {isSetParameters=true}
   &ShapeLocation(
    MaskSpawnGroup=false
    Override Seed=false
    Sphere Radius=&FloatFromCurve(FloatCurve=DataInterface() {propertyValues {Curve {keys [<at 0.33 1>,<at 1 0>]}}}, CurveIndex=$Transient.ParentNormalizedAge, Scale Curve=5)
    UseOwnerScale=true
   )
   &InheritVelocity(Speed Limit=true, UseSpeedThreshold=false)
   &AddVelocity(Velocity=&RandomRangeVector(Minimum=<5 5 15>, Maximum=<-5 -5 10>), Velocity Speed Scale=3)
   &KillParticles(Kill Particles=&SetBoolByFloatComparison(
     A=&FloatFromCurve(
      FloatCurve=DataInterface() {propertyValues {Curve {keys [
          {arriveTangent=25.7884, interpMode=RCIM_Cubic, time=0.33, value=0}
          {arriveTangent=-0.0504, interpMode=RCIM_Cubic, leaveTangent=-0.0504, tangentMode=RCTM_User, time=0.4, value=1}
          {arriveTangent=0.0831, interpMode=RCIM_Cubic, leaveTangent=0.0831, tangentMode=RCTM_User, time=0.6, value=1}
          {arriveTangent=-10.9519, interpMode=RCIM_Constant, tangentMode=RCTM_User, time=0.8001, value=0}
         ]}}}
      CurveIndex=$Transient.ParentNormalizedAge
     )
     B=&RandomRangeFloat(Minimum=0, Maximum=1)
     Comparison Type=::NewEnumerator6
    )) {module=KillParticles001}
  ]
  particleUpdateScript [
   &CameraOffset(Camera Offset Amount=-25)
   &ParticleState()
   &ScaleSpriteSize(Uniform Curve Sprite Scale=DataInterface() {propertyValues {Curve {keys [{interpMode=RCIM_Cubic, tangentMode=RCTM_User, time=0, value=1}, {interpMode=RCIM_Cubic, tangentMode=RCTM_User, time=1, value=0}]}}})
   &Drag(Drag=0.15, Rotational Drag=0.1, Use Linear Drag=true, Use Rotational Drag=true)
   &SolveForcesAndVelocity(Clamp Velocity=false, Limit Acceleration=false)
  ]
  rendererClasses [&NiagaraSpriteRendererProperties]
  renderers [
   {rendererClass=&NiagaraSpriteRendererProperties}
   {rendererClass=&NiagaraSpriteRendererProperties, rendererIndex=1}
  ]
 }
 Spline_Mesh_Interpolate_ShardGlow {
  bEnabled=false
  emitterUpdateScript [
   &EmitterState(
    Inactive Response=::NewEnumerator1
    Life Cycle Mode=::NewEnumerator1
    Loop Behavior=::NewEnumerator1
    Loop Duration Mode=::NewEnumerator1
    UseLoopDelay=false
   )
   &SpawnBurst_Instantaneous(Spawn Count=7, Use Spawn Probability=false, UseLoopCountLimit=false)
   /SetVariables(Emitter.Glow Amount:NiagaraFloat=&MakeCustomFloatFromBool(Bool=$User.IsFrontEnd, True Float=0.01, False Float=0.02)) {isSetParameters=true}
  ]
  particleSpawnScript [
   &PartitionParticles(Number Of Partitions=5)
   &InitializeParticle(
    Color=&MakeCustomLinearColorFromBool(Bool=$User.IsFrontEnd, True Color=<rgba 1.5 0.3762 0.1134 1>, False Color=<rgba 3 0.7525 0.2268 1>)
    Mesh Scale=&MakeVector(
     X=1
     Y=1
     Z=&Lerp_Float(
      A=&SelectFloatFromArray(
       Direct Array Index=$Particles.PartitionParticles.Partition
       Float Selection Array=DataInterface() {propertyValues {FloatData [1.35,1.2,1.2,0.8,0.8,0.5,0.5]}}
      )
      B=&SelectFloatFromArray(
       Direct Array Index=$Particles.PartitionParticles.Partition
       Float Selection Array=DataInterface() {propertyValues {FloatData [1.25,1.25,1.25,0.9,0.9,0.75,0.75]}}
      )
      Alpha=&MakeCustomFloatFromBool(Bool=$User.IsLargeBody, True Float=0, False Float=0)
     )
    )
    Mesh Scale Mode=::NewEnumerator4
    UsePositionOffset=false
    Write Material Random=true
   )
   &InitialMeshOrientation(Mesh Orientation Mode=::NewEnumerator1, Rotation=<0 0 0.3>, Use Rotation Vector=true)
   &CameraOffset()
  ]
  particleUpdateScript [
   &ParticleState(Kill Particles When Lifetime Has Elapsed=false, Let Infinitely Lived Particles Die When Emitter Deactivates=true)
   &DynamicMaterialParameters(
    Index 0 Param 1=&SelectFloatFromArray(
     Direct Array Index=$Particles.PartitionParticles.Partition
     Float Selection Array=DataInterface() {propertyValues {FloatData [50,40,40,30,30]}}
    )
    Index 3 Param 1=1
    Index 3 Param 2=1
    Index 3 Param 3=1
    Index 3 Param 4=1
    Module.Param0WriteEnabled=true
    Module.Param0WriteEnabled3=true
    Module.Param1WriteEnabled=true
    Module.Param1WriteEnabled3=true
    Module.Param2WriteEnabled=true
    Module.Param2WriteEnabled3=true
    Module.Param3WriteEnabled=true
    Module.Param3WriteEnabled3=true
    Write Parameter Index 3=true
   )
   &UpdateMeshOrientation(
    Facing Direction=&FN_GetSystemAxis(Axis:ENiagaraOrientationAxis=::XAxis, Flip Axis:NiagaraBool=false)
    Facing Interpolation=false
    Orientation Method=::NewEnumerator1
    Override Quat=false
    TransformFacing=false
    TransformUp=false
   ) {module=UpdateMeshOrientation001}
   &InterpolateOverTime(
    Rate Of Change:NiagaraFloat=&RandomRangeFloat(Minimum=2, Maximum=5)
    Target Vector Value:Vector3f=&ConvertPositionToVector(Input Position=$Particles.Position, Convert To Absolute Large World Vector=false)
    Variable Type=::NewEnumerator2
   )
   &SkeletalMeshLocation(
    Bone / Socket Index=&SelectIntFromArray(Direct Array Index=$Particles.PartitionParticles.Partition, Int Selection Array=DataInterface() {propertyValues {IntData [3,2,4,1,5,6,0]}})
    Bone Sampling Mode=::NewEnumerator4
    MaskSpawnGroup=false
    Orientation Sampling=::NewEnumerator1
    Sampled Position Offset=&MakeCustomVectorFromBool(Bool=$User.IsFrontEnd, True Vector=<0 0 -8>, False Vector=<1 0 -10>)
    Sampled Position Offset Coordinate Space:ENiagaraSkelMeshTransforms=::NewEnumerator1
    Skeletal Mesh=$System.Skeletal Mesh
   ) {module=SkeletalMeshLocation003}
   /SetVariables(Particles.Interpolate Dynamic=&MakeFloat4FromVectorAndFloat(
     "Vector (RGB)"=$StackContext.InterpolateOverTime.InterpolateOverTime_V3_.Moving Average
     "Float (Alpha)"=&Add_Float(
      A=&MakeCustomFloatFromBool(Bool=$User.IsFrontEnd, True Float=1, False Float=$User.Interpolate Intensity)
      B=0.5
     )
    )) {isSetParameters=true}
   /SetVariables(Particles.color2:LinearColor=&MultiplyLinearColor(
     Linear Color=&SelectLinearColorFromArray(
      Array Sampling Mode=::NewEnumerator0
      Direct Array Index=?
      Color Selection Array=DataInterface() {propertyValues {ColorData [<rgba 0 0.5283 1 1>, <rgba 0 0.0967 0.6302 1>]}}
      Random Seed=0
     )
     Scale Factor=<1.5 1.5 1.5 1>
    )) {enabled=false, isSetParameters=true}
  ]
  rendererClasses [&NiagaraMeshRendererProperties]
  renderers [{rendererClass=&NiagaraMeshRendererProperties, rendererIndex=0}]
 }
}
summary {
 systemName=NS_Aura_Fall_Visceral
 emitters [
  {bEnabled=false, emitterName=Spline_Mesh_Interpolate_ShardGlow, rendererClasses [&NiagaraMeshRendererProperties]}
  {emitterName=Body_Embers, rendererClasses [&NiagaraSpriteRendererProperties]}
  {emitterName=Falling_Leaf, rendererClasses [&NiagaraMeshRendererProperties]}
  {emitterName=LeafEmbers, rendererClasses [&NiagaraSpriteRendererProperties]}
  {emitterName=Glow, rendererClasses [&NiagaraSpriteRendererProperties]}
 ]
 userVariables [
  User.Interpolate Intensity {type=NiagaraFloat, defaultValue=3.5, description=""}
  User.IsFrontEnd {type=NiagaraBool, defaultValue=true}
  User.IsLargeBody {type=NiagaraBool}
 ]
}
